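# Material function that supplies cliff tile color, alpha and normal, named for lake island rocks.
# A static bool input, UseWorldAlign, defaulting to true, switches each output between world-aligned projection and plain UV sampling of the CliffTileB diffuse and normal textures.
# One TextureScale input serves as the world projection size in the aligned path and as a direct UV multiplier in the other, so its effect differs between modes.
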MaterialFunction /Figment_S09/Athena/Environments/Island/Materials/MF_IslandRocks_Lake
graph {nodes [
  n1 = FunctionInput() {bUsePreviewValueAsDefault=true, inputName=TextureScale, inputType=FunctionInput_Scalar, previewValue=<2500 0 0 1>}
  n2 = MaterialFunctionCall(
   TextureObject=TextureObject() {texture=/Game/Packages/DS_Fortnite_Terrain_NoLOD/Textures/T_CliffTileB_D}
   TextureSize=@n1
   Export Float 4=StaticBool() {value=true}
  ) {materialFunction=&WorldAlignedTexture}
  n3 = Multiply(A=TextureCoordinate(), B=@n1)
  n4 = TextureSample(UVs=@n3) {texture=&T_CliffTileB_D}
  n5 = FunctionInput(Preview=StaticBool() {value=true}) {bUsePreviewValueAsDefault=true, inputName=UseWorldAlign, inputType=FunctionInput_StaticBool}
  FunctionOutput(None=StaticSwitch(True=ComponentMask(None=@n2.XYZ Texture) {b=true, g=true, r=true}, False=@n4.RGB, Value=@n5)) {outputName=RGB}
  FunctionOutput(None=StaticSwitch(True=ComponentMask(None=@n2.XYZ Texture) {a=true}, False=@n4.A, Value=@n5)) {outputName=A, sortPriority=1}
  n6 = TextureObject() {samplerType=SAMPLERTYPE_Normal, texture=/Game/Packages/DS_Fortnite_Terrain_NoLOD/Textures/T_CliffTileB_N}
  FunctionOutput(None=StaticSwitch(
    True=MaterialFunctionCall(TextureObject=@n6, TextureSize=@n1) {materialFunction=&WorldAlignedNormal}.XYZ Texture
    False=TextureSample(UVs=@n3) {samplerType=SAMPLERTYPE_Normal, texture=&T_CliffTileB_N}.RGB
    Value=@n5
   )) {outputName=N, sortPriority=2}
 ]}
